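# Translucent, unlit, two-sided Niagara sprite material for the spiral in the StrawTurnHead explosion effect.
# The spiral texture's red channel tinted by particle color feeds an emissive adjustment function.
# Green minus a dynamic "Disolve" parameter, saturated and multiplied by particle alpha, gives erosion-style opacity through a depth fade.
# A separate sphere-gradient mask multiplied with green has no visible consumer in the graph.
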Material /StrawTurn/Gameplay/FX/M_StrawTurnHead_Explosion_Spiral
graph {
 properties {
  bAutomaticallySetUsageInEditor=false
  bUsedWithNiagaraSprites=true
  blendMode=BLEND_Translucent
  shadingModel=MSM_Unlit
  twoSided=true
  usedShadingModels=Unlit
 }
 nodes [
  n1 = ParticleColor()
  n2 = TextureSample(UVs=TextureCoordinate()) {texture=&/T_StrawTurnHead_Spiral}
  n3 = Subtract(A=@n2.G, B=DynamicParameter() {defaultValue=<rgba 0 1 1 1>, paramNames [Disolve,Param2,Param3,Param4]}.Disolve)
  DepthFade(Opacity=Multiply(A=@n1.A, B=Saturate(None=@n3)), FadeDistance=ScalarParameter() {defaultValue=999, parameterName=DepthFade})
  MaterialFunctionCall(Emissive=Multiply(A=@n1.RGB, B=@n2.R)) {materialFunction=&MF_FX_EmissiveAdjustments}
  n4 = Power(Base=MaterialFunctionCall!(UVs, CenterPosition, Radius=Constant() {r=0.5}) {materialFunction=/Engine/Functions/Engine_MaterialFunctions01/Gradient/SphereGradient-2D}.Result 0-1) {constExponent=9}
  Multiply(A=@n2.G, B=Saturate(None=Multiply(A=@n4) {constB=2}))
 ]
}
