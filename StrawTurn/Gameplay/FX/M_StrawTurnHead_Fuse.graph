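# Masked, lit Niagara ribbon material for the StrawTurn head's fuse, with an optional glowing burn band.
# Base color comes from a squared, contrast-boosted fuse texture tinted by particle color.
# HotFuseMask slides a sine band along the UVs, sharpened by Exp and tinted by boosted HotFuseColor, which UseLitFuse blends in.
# FuseMask offsets the UVs into an inverted floor, forming a hard cutoff mask for consuming the fuse.
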Material /StrawTurn/Gameplay/FX/M_StrawTurnHead_Fuse
graph {
 properties {bUsedWithNiagaraRibbons=true, blendMode=BLEND_Masked, pixelDepthOffsetMode=PDOM_Legacy, usedShadingModels=Default Lit}
 nodes [
  Constant() {r=1}
  Constant() {r=0.8}
  n1 = TextureSample() {texture=&/T_StrawTurnHead_Fuse_D_FX}
  n2 = Multiply(A=Multiply(A=@n1.RGB, B=@n1.RGB), B=ScalarParameter() {defaultValue=10, parameterName=Contrast})
  n3 = Multiply(A=@n2, B=ParticleColor().RGB)
  n4 = Multiply(A=VectorParameter() {parameterName=HotFuseColor, defaultValue=<rgba 1 0.0642 0 1>}.RGB) {constB=20}
  n5 = TextureCoordinate()
  n6 = DynamicParameter() {defaultValue=<rgba 0 0 0 1>, paramNames [FuseMask,HotFuseMask,UseLitFuse,Param4]}
  n7 = LinearInterpolate(Alpha=@n6.HotFuseMask) {constA=-0.5, constB=0.5}
  n8 = Power(
   Base=MaterialFunctionCall(Value=Add(A=@n5, B=@n7) {constB=0}, Period=Constant() {r=2}) {materialFunction=&LinearSine}.Linear Sine
   Exp=ScalarParameter() {defaultValue=10, parameterName=Exp}
  )
  LinearInterpolate(
   A=@n3
   B=Add(A=@n3, B=Multiply(A=@n4, B=@n8)) {constB=0}
   Alpha=DynamicParameter() {defaultValue=<rgba 0 0 0 1>, paramNames [FuseMask,HotFuseMask,UseLitFuse,Param4]}.UseLitFuse
  ) {constA=-0.5, constB=0.5}
  n9 = TextureCoordinate()
  OneMinus(None=Floor(None=Add(A=@n9, B=@n6.FuseMask) {constB=0}))
  Reroute(None=@n2)
 ]
}
