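# Niagara effect for a burning fuse on the StrawTurn head, attached via skeletal mesh sampling.
# Wick draws a ribbon along a Hermite spline, with pendulum sway and timeline-driven burn material parameters.
# Fuse (flame sprites, flickering light) and Sparks travel that spline on the eased system timeline, spawning when PlayFuse or Replicated is set.
# FuseStart and SparksStart ignition bursts have disabled spawn modules; DEBUG and Wick_ORIG emitters are disabled.
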NiagaraSystem /StrawTurn/Gameplay/FX/NS_StrawTurnHead_Fuse
dependencies {
 usedDataInterfaces [
  &NiagaraDataInterfaceCurve
  &NiagaraDataInterfaceArrayFloat3
  &NiagaraDataInterfaceArrayFloat
  &NiagaraDataInterfaceVectorCurve
  &NiagaraDataInterfaceVector2DCurve
  &NiagaraDataInterfaceSkeletalMesh
 ]
 usedDynamicInputs []
 usedModules [
  &self:SystemSpawnScript.SystemScriptSource.SystemScriptGraph.NiagaraNodeAssignment_0.SetVariables_#1
  &SystemState
  &Timeline
  &self:SystemSpawnScript.SystemScriptSource.SystemScriptGraph.NiagaraNodeAssignment_1.SetVariables_#2
 ]
}
emitters {
 DEBUG {
  bEnabled=false
  emitterUpdateScript [
   &EmitterState()
   &SpawnRate(
    Spawn Probability=&MakeCustomFloatFromBool(Bool=$User.Replicated, True Float=1, False Float=0)
    SpawnRate=10
    Use Spawn Probability=true
   )
  ]
  particleSpawnScript [&InitializeParticle(Color=<rgba 1 0 0.9427 1>, Sprite Size Mode=::NewEnumerator3, Uniform Sprite Size=25, UsePositionOffset=false, Write Material Random=true)]
  particleUpdateScript [
   &ParticleState()
   &SkeletalMeshLocation(MaskSpawnGroup=false, Mesh Sampling Type=::NewEnumerator5, Skeletal Mesh=$System.SkeletalMesh, Socket Sampling Mode=::NewEnumerator5)
   &ScaleColor(Scale RGB=&VectorFromFloat(Value=15), ScaleA=true, ScaleRGB=true)
  ]
  rendererClasses [&NiagaraSpriteRendererProperties]
  renderers [{rendererClass=&NiagaraSpriteRendererProperties}]
 }
 Fuse {
  emitterUpdateScript [
   &EmitterState(Life Cycle Mode=::NewEnumerator1, Loop Behavior=::NewEnumerator1, Loop Duration Mode=::NewEnumerator1, UseLoopDelay=false)
   &SpawnRate(
    Spawn Probability=&MakeCustomFloatFromBool(Bool=&Boolean_OrOperation(Bool1=$User.PlayFuse, Bool2=$User.Replicated), True Float=1, False Float=0)
    SpawnRate=60
    Use Spawn Probability=true
   )
  ]
  particleSpawnScript [
   &PartitionParticles()
   /SetVariables(Particles.VisibilityTag=$Particles.PartitionParticles.Partition) {isSetParameters=true, module=SetVariables_#3}
   &InitializeParticle(
    Lifetime=&MakeCustomFloatFromBool(
     Bool=&SetBoolByFloatComparison(A=0.5, B=&MakeFloatFromInt(Int=$Particles.PartitionParticles.Partition))
     True Float=&RandomRangeFloat(Minimum=0.1, Maximum=0.2)
     False Float=&RandomRangeFloat(Minimum=0.025, Maximum=0.1)
    )
    Position Offset=<0 0 12>
    Sprite Rotation Mode=::NewEnumerator2
    Sprite Size Mode=::NewEnumerator3
    Uniform Sprite Size=&MakeCustomFloatFromBool(
     Bool=&SetBoolByFloatComparison(A=0.5, B=&MakeFloatFromInt(Int=$Particles.PartitionParticles.Partition))
     True Float=50
     False Float=&RandomRangeFloat(Minimum=10, Maximum=25)
    )
    UsePositionOffset=false
    Write Material Random=true
   )
   &Color(Color=&MakeLinearColorFromVectorAndFloat(
     "Vector (RGB)"=&SelectVectorFromArray(
      Direct Array Index=$Particles.PartitionParticles.Partition
      Vector Selection Array=DataInterface() {propertyValues {FloatData [<5 5 5>, <5 1.3858 0.7867>]}}
     )
     "Float (Alpha)"=&SelectFloatFromArray(Direct Array Index=$Particles.PartitionParticles.Partition, Float Selection Array=DataInterface() {propertyValues {FloatData [0.1,1]}})
    ))
   &CameraOffset(Camera Offset Amount=5)
   /SetVariables(Particles.PreviousVisibility=1) {isSetParameters=true}
  ]
  particleUpdateScript [
   &ParticleState()
   &SkeletalMeshLocation(
    MaskSpawnGroup=false
    Mesh Sampling Type=::NewEnumerator5
    Sampled Orientation Coordinate Space=::Local
    Skeletal Mesh=$System.SkeletalMesh
    Socket Sampling Mode=::NewEnumerator5
   )
   /Game/Effects/Niagara/Modules/Curves/FN_HermiteSpline(
    Direction End:Vector3f=$System.HermiteDirectionEnd
    Direction Start:Vector3f=$System.HermiteDirectionStart
    Position Blend:NiagaraFloat=&Lerp_Float(
     A=1
     B=0
     Alpha=&FloatFromCurve(
      FloatCurve=DataInterface() {propertyValues {Curve {keys [
          {interpMode=RCIM_Cubic, tangentMode=RCTM_User, time=0, value=0}
          {arriveTangent=3.0981, interpMode=RCIM_Cubic, leaveTangent=3.0981, tangentMode=RCTM_User, time=1, value=1}
         ]}}}
      CurveIndex=$System.Timeline.NormalizedTime
     )
    )
    Position End:Vector3f=&AddVectorToPosition(Position=$Particles.SkeletalMeshLocation.SampledPosition, Vector=<15 -15 0>)
    Position Start:Vector3f=&AddVectorToPosition(Position=$Particles.SkeletalMeshLocation.SampledPosition, Vector=<0 0 0>)
    Write Particles Position=true
   )
   &InheritSourceMovement(
    Applied Position Scale=&Multiply_VectorByFloat(Vector=<-0.3 0.5 0.3>, Float=1)
    ApplyForce=false
    ApplyPosition=true
    ApplyPositionFalloff=false
   )
   &SolveForcesAndVelocity(Clamp Velocity=false, Limit Acceleration=false)
   &ScaleSpriteSize(
    Scale Sprite Size Mode=::NewEnumerator0
    Uniform Scale Factor=&MakeCustomFloatFromBool(
     Bool=&SetBoolByFloatComparison(A=0.5, B=&MakeFloatFromInt(Int=$Particles.PartitionParticles.Partition))
     True Float=1
     False Float=&FloatFromCurve(FloatCurve=DataInterface() {propertyValues {Curve {keys [{tangentMode=RCTM_User, time=0, value=0}, {tangentMode=RCTM_User, time=1, value=1}]}}})
    )
   )
   &ScaleColor(
    Scale Alpha=&FloatFromCurve(FloatCurve=DataInterface() {propertyValues {Curve {keys [
         {interpMode=RCIM_Cubic, tangentMode=RCTM_User, time=0, value=0}
         {interpMode=RCIM_Cubic, tangentMode=RCTM_User, time=0.5, value=1}
         {tangentMode=RCTM_User, time=1, value=0}
        ]}}})
    Scale RGB=&MakeCustomVectorFromBool(
     Bool=&SetBoolByFloatComparison(A=0.5, B=&MakeFloatFromInt(Int=$Particles.PartitionParticles.Partition))
     True Vector=&VectorFromCurve(
      VectorCurve=DataInterface() {propertyValues {
        XCurve {keys [<at 0 0.11>,<at 0.5 1>,<at 1 0.11>]}
        YCurve {keys [<at 0 0.05>,<at 0.5 0.13>,<at 1 0.05>]}
        ZCurve {keys [
          {arriveTangent=-0.3559, interpMode=RCIM_Cubic, leaveTangent=-0.3559, tangentMode=RCTM_User, time=0, value=0.03}
          {arriveTangent=0.6186, interpMode=RCIM_Cubic, leaveTangent=-0.8742, tangentMode=RCTM_Break, time=0.5, value=0}
          <at 1 0.03>
         ]}
       }}
      Scale Curve=<2 2 2>
     )
     False Vector=<1 1 1>
    )
    ScaleA=true
    ScaleRGB=true
   )
   &DynamicMaterialParameters(
    Index 0 Param 1=&FloatFromCurve(FloatCurve=DataInterface() {propertyValues {Curve {keys [<at 0 8>,<at 0.5 5>,<at 1 8>]}}})
    Index 0 Param 2=1
    Index 0 Param 3=1
    Index 0 Param 4=1
    Module.Index0Param0Write=true
    Module.Index0Param1Write=true
    Module.Index0Param2Write=true
    Module.Index0Param3Write=true
   )
   &self:ScratchModule(
    ConditionToSendEvent:NiagaraBool=&SetBoolByIntComparison(A=$Particles.VisibilityTag, B=$Particles.PreviousVisibility, Comparison Type=::NewEnumerator5)
    Event Generation Enabled:NiagaraBool=&SetBoolByFloatComparison(A=&Lerp_Float(A=1, B=0, Alpha=$System.Timeline001.NormalizedTime), B=0)
    Use Event Probability=false
   ) {module=ScratchModule}
   /SetVariables(Particles.PreviousVisibility:NiagaraInt32=$Particles.VisibilityTag) {isSetParameters=true}
   /SetVariables(Particles.FlickerWave:NiagaraFloat=&Waveform("Waveform [1]"=::NewEnumerator0, "[1] Phase Scale"=6, "[1] Frequency"=1, "[1] Amplitude Min/Max"=<0.8 1>)) {isSetParameters=true, module=SetVariables_#4}
   /Niagara/Modules/Update/SubUV/SubUVAnimation(Number Of Frames=&SelectFloatFromArray(Direct Array Index=$Particles.PartitionParticles.Partition, Float Selection Array=DataInterface() {propertyValues {FloatData [0,3]}}))
   &FN_Light_Attributes(
    Color=<rgba 1 0.1818 0 1>
    Create Color Attribute=true
    Create Exponent Attribute=true
    Create Radius Attribute=true
    Exponent=5
    Radius=&Multiply_Float(A=100, B=$Particles.FlickerWave)
    Scale Color=&Multiply_Float(A=2, B=$Particles.FlickerWave)
   )
  ]
  rendererClasses [&NiagaraSpriteRendererProperties, &NiagaraLightRendererProperties]
  renderers [
   {rendererClass=&NiagaraSpriteRendererProperties}
   {rendererClass=&NiagaraSpriteRendererProperties, rendererIndex=1}
   {rendererClass=&NiagaraLightRendererProperties, rendererIndex=2}
  ]
 }
 FuseStart {
  emitterUpdateScript [
   &EmitterState(Life Cycle Mode=::NewEnumerator1, Loop Behavior=::NewEnumerator1, Loop Duration Mode=::NewEnumerator1, UseLoopDelay=false)
   &SpawnBurst_Instantaneous(
    Spawn Count=2
    Spawn Probability=&MakeCustomFloatFromBool(Bool=$User.PlayFuse, True Float=1, False Float=0)
    Use Spawn Probability=true
    UseLoopCountLimit=false
   ) {enabled=false}
  ]
  particleSpawnScript [
   &PartitionParticles()
   /SetVariables(Particles.VisibilityTag=$Particles.PartitionParticles.Partition) {isSetParameters=true, module=SetVariables_#3}
   &InitializeParticle(
    Lifetime=&MakeCustomFloatFromBool(
     Bool=&SetBoolByFloatComparison(A=0.5, B=&MakeFloatFromInt(Int=$Particles.PartitionParticles.Partition))
     True Float=&RandomRangeFloat(Minimum=0.1, Maximum=0.2)
     False Float=&RandomRangeFloat(Minimum=0.025, Maximum=0.1)
    )
    Position Offset=<0 0 12>
    Sprite Rotation Mode=::NewEnumerator2
    Sprite Size Mode=::NewEnumerator3
    Uniform Sprite Size=&MakeCustomFloatFromBool(
     Bool=&SetBoolByFloatComparison(A=0.5, B=&MakeFloatFromInt(Int=$Particles.PartitionParticles.Partition))
     True Float=150
     False Float=&RandomRangeFloat(Minimum=25, Maximum=35)
    )
    UsePositionOffset=false
    Write Material Random=true
   )
   &Color(Color=&MakeLinearColorFromVectorAndFloat(
     "Vector (RGB)"=&SelectVectorFromArray(
      Direct Array Index=$Particles.PartitionParticles.Partition
      Vector Selection Array=DataInterface() {propertyValues {FloatData [<5 5 5>, <5 1.3858 0.7867>]}}
     )
     "Float (Alpha)"=&SelectFloatFromArray(Direct Array Index=$Particles.PartitionParticles.Partition, Float Selection Array=DataInterface() {propertyValues {FloatData [0.1,1]}})
    ))
   &SkeletalMeshLocation(
    MaskSpawnGroup=false
    Mesh Sampling Type=::NewEnumerator5
    Sampled Orientation Coordinate Space=::Local
    Skeletal Mesh=$System.SkeletalMesh
    Socket Sampling Mode=::NewEnumerator5
   )
   &FN_HermiteSpline(
    Direction End:Vector3f=$System.HermiteDirectionEnd
    Direction Start:Vector3f=$System.HermiteDirectionStart
    Position Blend:NiagaraFloat=&Lerp_Float(
     A=1
     B=0
     Alpha=&FloatFromCurve(
      FloatCurve=DataInterface() {propertyValues {Curve {keys [
          {interpMode=RCIM_Cubic, tangentMode=RCTM_User, time=0, value=0}
          {arriveTangent=3.0981, interpMode=RCIM_Cubic, leaveTangent=3.0981, tangentMode=RCTM_User, time=1, value=1}
         ]}}}
      CurveIndex=$System.Timeline.NormalizedTime
     )
    )
    Position End:Vector3f=&AddVectorToPosition(Position=$Particles.SkeletalMeshLocation.SampledPosition, Vector=<15 -15 0>)
    Position Start:Vector3f=&AddVectorToPosition(Position=$Particles.SkeletalMeshLocation.SampledPosition, Vector=<0 0 0>)
    Write Particles Position=true
   )
   &CameraOffset(Camera Offset Amount=5)
  ]
  particleUpdateScript [
   &ParticleState()
   &SolveForcesAndVelocity(Clamp Velocity=false, Limit Acceleration=false)
   &ScaleSpriteSize(
    Scale Sprite Size Mode=::NewEnumerator0
    Uniform Scale Factor=&MakeCustomFloatFromBool(
     Bool=&SetBoolByFloatComparison(A=0.5, B=&MakeFloatFromInt(Int=$Particles.PartitionParticles.Partition))
     True Float=1
     False Float=&FloatFromCurve(FloatCurve=DataInterface() {propertyValues {Curve {keys [{tangentMode=RCTM_User, time=0, value=0}, {tangentMode=RCTM_User, time=1, value=1}]}}})
    )
   )
   &ScaleColor(
    Scale Alpha=&FloatFromCurve(FloatCurve=DataInterface() {propertyValues {Curve {keys [
         {interpMode=RCIM_Cubic, tangentMode=RCTM_User, time=0, value=0}
         {interpMode=RCIM_Cubic, tangentMode=RCTM_User, time=0.5, value=1}
         {tangentMode=RCTM_User, time=1, value=0}
        ]}}})
    Scale RGB=&MakeCustomVectorFromBool(
     Bool=&SetBoolByFloatComparison(A=0.5, B=&MakeFloatFromInt(Int=$Particles.PartitionParticles.Partition))
     True Vector=&VectorFromCurve(
      VectorCurve=DataInterface() {propertyValues {
        XCurve {keys [<at 0 0.11>,<at 0.5 1>,<at 1 0.11>]}
        YCurve {keys [<at 0 0.05>,<at 0.5 0.13>,<at 1 0.05>]}
        ZCurve {keys [
          {arriveTangent=-0.3559, interpMode=RCIM_Cubic, leaveTangent=-0.3559, tangentMode=RCTM_User, time=0, value=0.03}
          {arriveTangent=0.6186, interpMode=RCIM_Cubic, leaveTangent=-0.8742, tangentMode=RCTM_Break, time=0.5, value=0}
          <at 1 0.03>
         ]}
       }}
      Scale Curve=<10 10 10>
     )
     False Vector=<5 5 5>
    )
    ScaleA=true
    ScaleRGB=true
   )
   &DynamicMaterialParameters(
    Index 0 Param 1=&FloatFromCurve(FloatCurve=DataInterface() {propertyValues {Curve {keys [<at 0 8>,<at 0.5 5>,<at 1 8>]}}})
    Index 0 Param 2=1
    Index 0 Param 3=1
    Index 0 Param 4=1
    Module.Index0Param0Write=true
    Module.Index0Param1Write=true
    Module.Index0Param2Write=true
    Module.Index0Param3Write=true
   )
   /SetVariables(Particles.FlickerWave:NiagaraFloat=&Waveform("Waveform [1]"=::NewEnumerator0, "[1] Phase Scale"=6, "[1] Frequency"=1, "[1] Amplitude Min/Max"=<0.8 1>)) {isSetParameters=true, module=SetVariables_#4}
   &SubUVAnimation(Number Of Frames=&SelectFloatFromArray(Direct Array Index=$Particles.PartitionParticles.Partition, Float Selection Array=DataInterface() {propertyValues {FloatData [0,3]}}))
   &FN_Light_Attributes(
    Color=<rgba 1 0.1818 0 1>
    Create Color Attribute=true
    Create Exponent Attribute=true
    Create Radius Attribute=true
    Exponent=5
    Radius=&Multiply_Float(A=150, B=$Particles.FlickerWave)
    Scale Color=&Multiply_Float(A=4, B=$Particles.FlickerWave)
   )
  ]
  rendererClasses [&NiagaraSpriteRendererProperties, &NiagaraLightRendererProperties]
  renderers [
   {rendererClass=&NiagaraSpriteRendererProperties}
   {rendererClass=&NiagaraSpriteRendererProperties, rendererIndex=1}
   {rendererClass=&NiagaraLightRendererProperties, rendererIndex=2}
  ]
 }
 Sparks {
  emitterUpdateScript [
   &EmitterState(Life Cycle Mode=::NewEnumerator1, Loop Behavior=::NewEnumerator1, Loop Duration Mode=::NewEnumerator1, UseLoopDelay=false)
   &SpawnRate(
    Spawn Probability=&MakeCustomFloatFromBool(Bool=&Boolean_OrOperation(Bool1=$User.PlayFuse, Bool2=$User.Replicated), True Float=1, False Float=0)
    SpawnRate=35
    Use Spawn Probability=true
   )
  ]
  particleSpawnScript [
   &InitializeParticle(
    Lifetime=&MaskFloatBySpawnGroup(
     Masked Float=&RandomRangeFloat(Minimum=0.1, Maximum=0.3)
     Passthrough Float=&RandomRangeFloat(Minimum=0.025, Maximum=0.1)
    )
    Position Offset=<0 0 50>
    Sprite Size=&RandomRangeVector2D(Minimum=<1 1>, Maximum=<1 1>)
    Sprite Size Mode=::NewEnumerator4
    UsePositionOffset=false
    Write Material Random=true
   )
   &AddVelocity(Velocity=&Multiply_VectorByFloat(
     Vector=&RandomRangeVector(Minimum=<-250 -250 -250>, Maximum=<250 250 250>)
     Float=&Lerp_Float(
      A=1
      B=3
      Alpha=&FloatFromCurve(
       FloatCurve=DataInterface() {propertyValues {Curve {keys [
           {interpMode=RCIM_Cubic, tangentMode=RCTM_User, time=0, value=0}
           {arriveTangent=0.907, interpMode=RCIM_Cubic, leaveTangent=0.907, tangentMode=RCTM_User, time=0.8, value=0.2}
           {arriveTangentWeight=0.4864, interpMode=RCIM_Cubic, tangentMode=RCTM_User, time=1, value=1}
          ]}}}
       CurveIndex=&RandomRangeFloat(Minimum=0, Maximum=1)
      )
     )
    ))
   &Color(Color=&MakeLinearColorFromVectorAndFloat("Vector (RGB)"=<10 5.3681 0.47>, "Float (Alpha)"=&ClampFloat(Float=1)))
   &InheritVelocity(Inherited Velocity Amount Scale=<0.2 0.2 0.2>, Inherited Velocity Speed Limit=0, Speed Limit=false, UseSpeedThreshold=false)
   &SkeletalMeshLocation(
    MaskSpawnGroup=false
    Mesh Sampling Type=::NewEnumerator5
    Sampled Orientation Coordinate Space=::Local
    Skeletal Mesh=$System.SkeletalMesh
    Socket Sampling Mode=::NewEnumerator5
   )
   &FN_HermiteSpline(
    Direction End:Vector3f=$System.HermiteDirectionEnd
    Direction Start:Vector3f=$System.HermiteDirectionStart
    Position Blend:NiagaraFloat=&Lerp_Float(
     A=1
     B=0
     Alpha=&FloatFromCurve(
      FloatCurve=DataInterface() {propertyValues {Curve {keys [
          {interpMode=RCIM_Cubic, tangentMode=RCTM_User, time=0, value=0}
          {arriveTangent=3.0981, interpMode=RCIM_Cubic, leaveTangent=3.0981, tangentMode=RCTM_User, time=1, value=1}
         ]}}}
      CurveIndex=$System.Timeline.NormalizedTime
     )
    )
    Position End:Vector3f=&AddVectorToPosition(Position=$Particles.SkeletalMeshLocation.SampledPosition, Vector=<15 -15 0>)
    Position Start:Vector3f=&AddVectorToPosition(Position=$Particles.SkeletalMeshLocation.SampledPosition, Vector=<0 0 0>)
    Write Particles Position=true
   )
  ]
  particleUpdateScript [
   &ParticleState()
   &ScaleSpriteSize(
    Scale Factor=&Vector2DFromCurve(
     Vector2Curve=DataInterface() {propertyValues {
       XCurve {keys [<at 0 1 RCIM_Cubic>, {interpMode=RCIM_Cubic, leaveTangent=-1.9252, time=1, value=0.2}]}
       YCurve {keys [<at 0 3 RCIM_Cubic>,<at 1 0 RCIM_Cubic>]}
      }}
     ScaleCurve=<1 1>
    )
    Scale Sprite Size Mode=::NewEnumerator2
   )
   &AccelerationForce(Acceleration=&RandomRangeVector(Minimum=<0 0 -800>, Maximum=<1 1 -980>))
   &Drag(Drag=3, Use Linear Drag=true, Use Rotational Drag=false)
   &SolveForcesAndVelocity(Clamp Velocity=false, Limit Acceleration=false)
  ]
  rendererClasses [&NiagaraSpriteRendererProperties]
  renderers [{rendererClass=&NiagaraSpriteRendererProperties}]
 }
 SparksStart {
  emitterUpdateScript [
   &EmitterState(Life Cycle Mode=::NewEnumerator1, Loop Duration=0.1, UseLoopDelay=false)
   &SpawnBurst_Instantaneous(
    Spawn Count=25
    Spawn Probability=&MakeCustomFloatFromBool(Bool=$User.PlayFuse, True Float=1, False Float=0)
    Use Spawn Probability=true
    UseLoopCountLimit=true
   ) {enabled=false}
  ]
  particleSpawnScript [
   &InitializeParticle(
    Lifetime=&MaskFloatBySpawnGroup(
     Masked Float=&RandomRangeFloat(Minimum=0.1, Maximum=0.3)
     Passthrough Float=&RandomRangeFloat(Minimum=0.025, Maximum=0.1)
    )
    Position Offset=<0 0 50>
    Sprite Size=&RandomRangeVector2D(Minimum=<1 1>, Maximum=<1 1>)
    Sprite Size Mode=::NewEnumerator4
    UsePositionOffset=false
    Write Material Random=true
   )
   &AddVelocity(Velocity=&Multiply_VectorByFloat(
     Vector=&RandomRangeVector(Minimum=<-250 -250 -250>, Maximum=<250 250 250>)
     Float=&Lerp_Float(
      A=1
      B=3
      Alpha=&FloatFromCurve(
       FloatCurve=DataInterface() {propertyValues {Curve {keys [
           {interpMode=RCIM_Cubic, tangentMode=RCTM_User, time=0, value=0}
           {arriveTangent=0.907, interpMode=RCIM_Cubic, leaveTangent=0.907, tangentMode=RCTM_User, time=0.8, value=0.2}
           {arriveTangentWeight=0.4864, interpMode=RCIM_Cubic, tangentMode=RCTM_User, time=1, value=1}
          ]}}}
       CurveIndex=&RandomRangeFloat(Minimum=0, Maximum=1)
      )
     )
    ))
   &Color(Color=&MakeLinearColorFromVectorAndFloat("Vector (RGB)"=<10 5.3681 0.47>, "Float (Alpha)"=&ClampFloat(Float=1)))
   &InheritVelocity(Inherited Velocity Amount Scale=<0.2 0.2 0.2>, Inherited Velocity Speed Limit=0, Speed Limit=false, UseSpeedThreshold=false)
   &SkeletalMeshLocation(
    MaskSpawnGroup=false
    Mesh Sampling Type=::NewEnumerator5
    Sampled Orientation Coordinate Space=::Local
    Skeletal Mesh=$System.SkeletalMesh
    Socket Sampling Mode=::NewEnumerator5
   )
   &FN_HermiteSpline(
    Direction End:Vector3f=$System.HermiteDirectionEnd
    Direction Start:Vector3f=$System.HermiteDirectionStart
    Position Blend:NiagaraFloat=&Lerp_Float(
     A=1
     B=0
     Alpha=&FloatFromCurve(
      FloatCurve=DataInterface() {propertyValues {Curve {keys [
          {interpMode=RCIM_Cubic, tangentMode=RCTM_User, time=0, value=0}
          {arriveTangent=3.0981, interpMode=RCIM_Cubic, leaveTangent=3.0981, tangentMode=RCTM_User, time=1, value=1}
         ]}}}
      CurveIndex=$System.Timeline.NormalizedTime
     )
    )
    Position End:Vector3f=&AddVectorToPosition(Position=$Particles.SkeletalMeshLocation.SampledPosition, Vector=<15 -15 0>)
    Position Start:Vector3f=&AddVectorToPosition(Position=$Particles.SkeletalMeshLocation.SampledPosition, Vector=<0 0 0>)
    Write Particles Position=true
   )
  ]
  particleUpdateScript [
   &ParticleState()
   &ScaleSpriteSize(
    Scale Factor=&Vector2DFromCurve(
     Vector2Curve=DataInterface() {propertyValues {
       XCurve {keys [<at 0 1 RCIM_Cubic>, {interpMode=RCIM_Cubic, leaveTangent=-1.9252, time=1, value=0.2}]}
       YCurve {keys [<at 0 3 RCIM_Cubic>,<at 1 0 RCIM_Cubic>]}
      }}
     ScaleCurve=<1 1>
    )
    Scale Sprite Size Mode=::NewEnumerator2
   )
   &AccelerationForce(Acceleration=&RandomRangeVector(Minimum=<0 0 -800>, Maximum=<1 1 -980>))
   &Drag(Drag=3, Use Linear Drag=true, Use Rotational Drag=false)
   &SolveForcesAndVelocity(Clamp Velocity=false, Limit Acceleration=false)
   &ScaleColor(Scale RGB=&VectorFromFloat(Value=10), ScaleA=false, ScaleRGB=true)
  ]
  rendererClasses [&NiagaraSpriteRendererProperties]
  renderers [{rendererClass=&NiagaraSpriteRendererProperties}]
 }
 Wick {
  emitterUpdateScript [
   &EmitterState(Life Cycle Mode=::NewEnumerator1, Loop Behavior=::NewEnumerator1, Loop Duration Mode=::NewEnumerator1, UseLoopDelay=false)
   &SpawnBurst_Instantaneous(Spawn Count=25, Use Spawn Probability=false, UseLoopCountLimit=false)
  ]
  particleSpawnScript [
   &InitializeParticle(
    Color=<rgba 1 0.3942 0 1>
    Lifetime=5
    Ribbon Width=0.8
    Ribbon Width Mode=::NewEnumerator1
    UsePositionOffset=false
    Write Material Random=true
   )
   &PartitionParticles(Number Of Partitions=1, Particles Per Partition=25, Partition Mode:ENiagara_PartitionMode=::NewEnumerator1)
  ]
  particleUpdateScript [
   &ParticleState(Kill Particles When Lifetime Has Elapsed=false, Let Infinitely Lived Particles Die When Emitter Deactivates=true)
   &SkeletalMeshLocation(
    MaskSpawnGroup=false
    Mesh Sampling Type=::NewEnumerator5
    Sampled Orientation Coordinate Space=::Local
    Skeletal Mesh=$System.SkeletalMesh
    Socket Sampling Mode=::NewEnumerator5
   )
   &FN_HermiteSpline(
    Direction End:Vector3f=$System.HermiteDirectionEnd
    Direction Start:Vector3f=$System.HermiteDirectionStart
    Position Blend:NiagaraFloat=&OneMinusFloat(Float=$Particles.PartitionParticles.Normalized Index In Partition)
    Position End:Vector3f=&AddVectorToPosition(Position=$Particles.SkeletalMeshLocation.SampledPosition, Vector=<15 -15 0>)
    Position Start:Vector3f=&AddVectorToPosition(Position=$Particles.SkeletalMeshLocation.SampledPosition, Vector=<0 0 0>)
    Write Particles Position=true
   )
   &InheritSourceMovement(
    Applied Position Scale=&Multiply_VectorByFloat(
     Vector=<-0.3 0.5 0.3>
     Float=&FloatFromCurve(
      FloatCurve=DataInterface() {propertyValues {Curve {keys [
          {interpMode=RCIM_Cubic, tangentMode=RCTM_User, time=0, value=0}
          {arriveTangent=3.3731, interpMode=RCIM_Cubic, leaveTangent=3.3731, tangentMode=RCTM_User, time=1, value=1}
         ]}}}
      CurveIndex=&OneMinusFloat(Float=$Particles.PartitionParticles.Normalized Index In Partition)
     )
    )
    ApplyForce=false
    ApplyPosition=true
    ApplyPositionFalloff=false
   )
   &PendulumSetup(
    Module.Enable Spring Driver=false
    Pendulum Length:NiagaraFloat=&FloatFromCurve(
     FloatCurve=DataInterface() {propertyValues {Curve {keys [
         {interpMode=RCIM_Cubic, tangentMode=RCTM_User, time=0, value=0}
         {arriveTangent=2.3056, arriveTangentWeight=1.1759, interpMode=RCIM_Cubic, leaveTangent=2.3056, tangentMode=RCTM_User, time=1, value=1}
        ]}}}
     CurveIndex=&OneMinusFloat(Float=$Particles.PartitionParticles.Normalized Index In Partition)
    )
    Pendulum Pivot:NiagaraPosition=$Particles.Position
   )
   &SolveForcesAndVelocity(Clamp Velocity=false, Limit Acceleration=false)
   &PendulumConstraint(Calculate and Output Potential Energy=true)
   &CalculateAccurateVelocity(Enabled:NiagaraBool=$User.PlayFuse)
   &ScaleColor(Scale RGB=&VectorFromFloat(Value=1), ScaleA=false, ScaleRGB=true)
   &DynamicMaterialParameters(
    Index 0 Param 1=&Lerp_Float(
     A=&MakeCustomFloatFromBool(Bool=$User.Lit, True Float=1, False Float=0)
     B=1.3
     Alpha=&FloatFromCurve(
      FloatCurve=DataInterface() {propertyValues {Curve {keys [
          {interpMode=RCIM_Cubic, tangentMode=RCTM_User, time=0, value=0}
          {arriveTangent=3.0981, interpMode=RCIM_Cubic, leaveTangent=3.0981, tangentMode=RCTM_User, time=1, value=1}
         ]}}}
      CurveIndex=$System.Timeline.NormalizedTime
     )
    )
    Index 0 Param 2=&Lerp_Float(
     A=0
     B=1
     Alpha=&FloatFromCurve(
      FloatCurve=DataInterface() {propertyValues {Curve {keys [
          {interpMode=RCIM_Cubic, tangentMode=RCTM_User, time=0, value=0}
          {arriveTangent=3.0981, interpMode=RCIM_Cubic, leaveTangent=3.0981, tangentMode=RCTM_User, time=1, value=1}
         ]}}}
      CurveIndex=$System.Timeline.NormalizedTime
     )
    )
    Index 0 Param 3=&MakeCustomFloatFromBool(Bool=$User.PlayFuse, True Float=1, False Float=0)
    Module.Index0Param0Write=true
    Module.Index0Param1Write=true
    Module.Index0Param2Write=true
    Module.Index0Param3Write=true
   )
  ]
  rendererClasses [&NiagaraRibbonRendererProperties]
  renderers [{rendererClass=&NiagaraRibbonRendererProperties}]
 }
 Wick_ORIG {
  bEnabled=false
  emitterSpawnScript [/Game/Effects/Niagara/Modules/Partition/FN_PartitionAttributes(Number of Bone Sets=1, Particles Per Partition=3)]
  emitterUpdateScript [
   &EmitterState(Life Cycle Mode=::NewEnumerator1, Loop Behavior=::NewEnumerator1, Loop Duration Mode=::NewEnumerator1, UseLoopDelay=false)
   &SpawnBurst_Instantaneous(Spawn Count=$Emitter.FN_PartitionAttributes.Spawn Total, Use Spawn Probability=false)
  ]
  particleSpawnScript [
   &InitializeParticle(
    Color=<rgba 1 0.3942 0 1>
    Lifetime=5
    Ribbon Width=0.8
    Ribbon Width Mode=::NewEnumerator1
    UsePositionOffset=false
    Write Material Random=true
   )
   /Game/Effects/Niagara/Modules/Partition/FN_Particle_partition_v5/FN_Particle_Partition_v5(Enable Ribbon ID=true)
  ]
  particleUpdateScript [
   &ParticleState(Kill Particles When Lifetime Has Elapsed=false, Let Infinitely Lived Particles Die When Emitter Deactivates=true)
   &SkeletalMeshLocation(
    MaskSpawnGroup=false
    Mesh Sampling Type=::NewEnumerator5
    Sampled Orientation Coordinate Space=::Local
    Skeletal Mesh=$System.SkeletalMesh
    Socket Sampling Mode=::NewEnumerator5
   )
   &FN_HermiteSpline(
    Direction End:Vector3f=$System.HermiteDirectionEnd
    Direction Start:Vector3f=$System.HermiteDirectionStart
    Position Blend:NiagaraFloat=$Particles.Initial.Partition.NormalizedID
    Position End:Vector3f=&AddVectorToPosition(Position=$Particles.SkeletalMeshLocation.SampledPosition, Vector=<15 -15 0>)
    Position Start:Vector3f=&AddVectorToPosition(Position=$Particles.SkeletalMeshLocation.SampledPosition, Vector=<0 0 0>)
    Write Particles Position=true
   )
   &InheritSourceMovement(
    Applied Position Scale=&Multiply_VectorByFloat(
     Vector=<-0.3 0.5 0.3>
     Float=&FloatFromCurve(
      FloatCurve=DataInterface() {propertyValues {Curve {keys [
          {interpMode=RCIM_Cubic, tangentMode=RCTM_User, time=0, value=0}
          {arriveTangent=3.3731, interpMode=RCIM_Cubic, leaveTangent=3.3731, tangentMode=RCTM_User, time=1, value=1}
         ]}}}
      CurveIndex=$Particles.Partition.NormalizedID
     )
    )
    ApplyForce=false
    ApplyPosition=true
    ApplyPositionFalloff=false
   )
   &PendulumSetup(
    Module.Enable Spring Driver=false
    Pendulum Length:NiagaraFloat=&FloatFromCurve(
     FloatCurve=DataInterface() {propertyValues {Curve {keys [
         {interpMode=RCIM_Cubic, tangentMode=RCTM_User, time=0, value=0}
         {arriveTangent=2.3056, arriveTangentWeight=1.1759, interpMode=RCIM_Cubic, leaveTangent=2.3056, tangentMode=RCTM_User, time=1, value=1}
        ]}}}
     CurveIndex=$Particles.Partition.NormalizedID
    )
    Pendulum Pivot:NiagaraPosition=$Particles.Position
   )
   &SolveForcesAndVelocity(Clamp Velocity=false, Limit Acceleration=false)
   &PendulumConstraint(Calculate and Output Potential Energy=true)
   &CalculateAccurateVelocity(Enabled:NiagaraBool=$User.PlayFuse)
   &ScaleColor(Scale RGB=&VectorFromFloat(Value=1), ScaleA=false, ScaleRGB=true)
   &DynamicMaterialParameters(
    Index 0 Param 1=&Lerp_Float(
     A=&MakeCustomFloatFromBool(Bool=$User.Lit, True Float=1, False Float=0)
     B=1.3
     Alpha=&FloatFromCurve(
      FloatCurve=DataInterface() {propertyValues {Curve {keys [
          {interpMode=RCIM_Cubic, tangentMode=RCTM_User, time=0, value=0}
          {arriveTangent=3.0981, interpMode=RCIM_Cubic, leaveTangent=3.0981, tangentMode=RCTM_User, time=1, value=1}
         ]}}}
      CurveIndex=$System.Timeline.NormalizedTime
     )
    )
    Index 0 Param 2=&Lerp_Float(
     A=0
     B=1
     Alpha=&FloatFromCurve(
      FloatCurve=DataInterface() {propertyValues {Curve {keys [
          {interpMode=RCIM_Cubic, tangentMode=RCTM_User, time=0, value=0}
          {arriveTangent=3.0981, interpMode=RCIM_Cubic, leaveTangent=3.0981, tangentMode=RCTM_User, time=1, value=1}
         ]}}}
      CurveIndex=$System.Timeline.NormalizedTime
     )
    )
    Index 0 Param 3=&MakeCustomFloatFromBool(Bool=$User.PlayFuse, True Float=1, False Float=0)
    Module.Index0Param0Write=true
    Module.Index0Param1Write=true
    Module.Index0Param2Write=true
    Module.Index0Param3Write=true
   )
   &DynamicMaterialParameters(
    Index 0 Param 1=&Lerp_Float(
     A=0
     B=1.3
     Alpha=&FloatFromCurve(
      FloatCurve=DataInterface() {propertyValues {Curve {keys [
          {interpMode=RCIM_Cubic, tangentMode=RCTM_User, time=0, value=0}
          {arriveTangent=3.0981, interpMode=RCIM_Cubic, leaveTangent=3.0981, tangentMode=RCTM_User, time=1, value=1}
         ]}}}
      CurveIndex=$System.Timeline.NormalizedTime
     )
    )
    Index 0 Param 2=&Lerp_Float(
     A=0
     B=1
     Alpha=&FloatFromCurve(
      FloatCurve=DataInterface() {propertyValues {Curve {keys [
          {interpMode=RCIM_Cubic, tangentMode=RCTM_User, time=0, value=0}
          {arriveTangent=3.0981, interpMode=RCIM_Cubic, leaveTangent=3.0981, tangentMode=RCTM_User, time=1, value=1}
         ]}}}
      CurveIndex=$System.Timeline.NormalizedTime
     )
    )
    Index 0 Param 3=&MakeCustomFloatFromBool(Bool=$User.PlayFuse, True Float=1, False Float=0)
    Module.Index0Param0Write=true
    Module.Index0Param1Write=true
    Module.Index0Param2Write=true
    Module.Index0Param3Write=true
   ) {enabled=false, module=DynamicMaterialParameters001}
  ]
  rendererClasses [&NiagaraRibbonRendererProperties]
  renderers [
   {rendererClass=&NiagaraRibbonRendererProperties}
   {rendererClass=&NiagaraRibbonRendererProperties, rendererIndex=1}
  ]
 }
}
summary {
 systemName=NS_StrawTurnHead_Fuse
 emitters [
  {bEnabled=false, emitterName=Wick_ORIG, rendererClasses [&NiagaraRibbonRendererProperties]}
  {emitterName=Fuse, rendererClasses [&NiagaraSpriteRendererProperties, &NiagaraLightRendererProperties]}
  {emitterName=Sparks, rendererClasses [&NiagaraSpriteRendererProperties]}
  {emitterName=FuseStart, rendererClasses [&NiagaraSpriteRendererProperties, &NiagaraLightRendererProperties]}
  {emitterName=SparksStart, rendererClasses [&NiagaraSpriteRendererProperties]}
  {emitterName=Wick, rendererClasses [&NiagaraRibbonRendererProperties]}
  {bEnabled=false, emitterName=DEBUG, rendererClasses [&NiagaraSpriteRendererProperties]}
 ]
 userVariables [
  User.FuseTime {type=NiagaraFloat, defaultValue=120, description=""}
  User.Lit {type=NiagaraBool}
  User.PlayFuse {type=NiagaraBool}
  User.Replicated {type=NiagaraBool}
 ]
}
